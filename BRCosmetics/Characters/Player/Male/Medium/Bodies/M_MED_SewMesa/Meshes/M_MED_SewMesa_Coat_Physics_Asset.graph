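# Physics asset for the SewMesa male medium cosmetic coat, providing capsule bodies and joints for physics motion of the jacket's dyn_jacket bones.
# Tiny base bodies, spine and thighs lock angularly to the pelvis, with large pelvis, spine and thigh capsules forming the body volume.
# Side and front chains (two segments) and back chains (one segment) hang from those bases with tight swing and twist limits.
# An extra locked constraint ties the center back base to the left back segment.
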PhysicsAsset /BRCosmetics/Characters/Player/Male/Medium/Bodies/M_MED_SewMesa/Meshes/M_MED_SewMesa_Coat_Physics_Asset
bodies {
 dyn_jacket_01_L [{
   shapeType=Capsule
   length=4.2013
   radius=3.3732
   shapeName=dyn_jacket_01_L_capsule
   center=<6.6786 -1.9 -0.4602>
   rotation=<rot 2.5618 -90 6.5981>
  }]
 dyn_jacket_01_R [{
   shapeType=Capsule
   length=4.2013
   radius=3.3732
   shapeName=dyn_jacket_01_L_capsule
   center=<6.6786 -1.6 -0.4348>
   rotation=<rot 2.5618 -90 -3.4018>
  }]
 dyn_jacket_02_L [{shapeType=Capsule, length=8, radius=2.3788, shapeName=dyn_jacket_02_L_capsule, center=<4 -2 0>, rotation=<rot 0 90 -5>}]
 dyn_jacket_02_R [{shapeType=Capsule, length=8, radius=2.37, shapeName=dyn_jacket_02_L_capsule, center=<4 -2 2>, rotation=<rot 0 90 2>}]
 dyn_jacket_back_02_C [{shapeType=Capsule, length=7.923, radius=3.3937, shapeName=dyn_jacket_back_02_C_capsule, center=<3 -2 0>}]
 dyn_jacket_back_02_L [{shapeType=Capsule, length=6, radius=3, shapeName=dyn_jacket_back_02_L_capsule, center=<2.8981 -1.1667 0.1423>, rotation=<rot 0 90.0001 0>}]
 dyn_jacket_back_02_R [{shapeType=Capsule, length=6, radius=3, shapeName=dyn_jacket_back_02_R_capsule, center=<3 -1.8459 0.1>, rotation=<rot -1.8145 90 0>}]
 dyn_jacket_back_base_C [{shapeType=Capsule, length=0.505, radius=0.505, shapeName=dyn_jacket_back_base_C_capsule, rotation=<rot 0 -90 0>}]
 dyn_jacket_back_base_L [{shapeType=Capsule, length=0.505, radius=0.505, shapeName=dyn_jacket_back_base_L_capsule, rotation=<rot 0 -90 0>}]
 dyn_jacket_back_base_R [{shapeType=Capsule, length=0.505, radius=0.505, shapeName=dyn_jacket_back_base_R_capsule, rotation=<rot 0 -90 0>}]
 dyn_jacket_base_L [{shapeType=Capsule, length=0.505, radius=0.505, shapeName=dyn_jacket_base_L_capsule, rotation=<rot 0 -90 0>}]
 dyn_jacket_base_R [{shapeType=Capsule, length=0.505, radius=0.505, shapeName=dyn_jacket_base_R_capsule, rotation=<rot 0 -90 0>}]
 dyn_jacket_front_01_L [{
   shapeType=Capsule
   length=4.7949
   radius=2.7216
   shapeName=dyn_jacket_front_01_L_capsule
   center=<6.8802 -2.5 0>
   rotation=<rot 7.0371 -90 0.1892>
  }]
 dyn_jacket_front_01_R [{
   shapeType=Capsule
   length=5.173
   radius=2.7216
   shapeName=dyn_jacket_front_01_L_capsule
   center=<6.8802 -2.5 -0.0996>
   rotation=<rot 7.0371 -90 0.1892>
  }]
 dyn_jacket_front_02_L [{
   shapeType=Capsule
   length=8
   radius=3.2631
   shapeName=dyn_jacket_front_02_L_capsule
   center=<6 -2 0>
   rotation=<rot 10.0402 83.8207 -11.4441>
  }]
 dyn_jacket_front_02_R [{shapeType=Capsule, length=9, radius=3.2631, shapeName=dyn_jacket_front_02_L_capsule, center=<6 -2 0>, rotation=<rot -20 90 12>}]
 dyn_jacket_front_base_L [{shapeType=Capsule, length=0.505, radius=0.505, shapeName=dyn_jacket_front_base_L_capsule, rotation=<rot 0 -90 0>}]
 dyn_jacket_front_base_R [{shapeType=Capsule, length=0.505, radius=0.505, shapeName=dyn_jacket_front_base_R_capsule, rotation=<rot 0 -90 0>}]
 pelvis [{shapeType=Capsule, length=9, radius=10.5, shapeName=pelvis_capsule, center=<1.2394 -0.568 0.0758>}]
 spine_01 [{shapeType=Capsule, length=12, radius=13, shapeName=spine_01_capsule, center=<8 -0.0014 -0.2579>, rotation=<rot 4.5707 90 -90>}]
 thigh_l [{
   shapeType=Capsule
   length=25
   radius=7.9
   shapeName=thigh_l_capsule
   center=<-11.8485 1.4288 0.3357>
   rotation=<rot -0.2652 -89.9398 -88.7579>
  }]
 thigh_r [{
   shapeType=Capsule
   length=25.368
   radius=7.382
   shapeName=thigh_r_capsule
   center=<10.2917 -1.4228 0.3045>
   rotation=<rot 0.158 -91.1937 -88.7398>
  }]
}
constraints [
 {
  bone1Name=dyn_jacket_back_base_C
  bone2Name=pelvis
  swing1LimitDegrees=45
  swing1Motion=Locked
  swing2LimitDegrees=45
  swing2Motion=Locked
  twistLimitDegrees=45
  twistMotion=Locked
 }
 {bone1Name=dyn_jacket_back_02_C, bone2Name=dyn_jacket_back_base_C, swing1LimitDegrees=20, swing2LimitDegrees=10, twistLimitDegrees=5}
 {
  bone1Name=dyn_jacket_back_base_L
  bone2Name=pelvis
  swing1LimitDegrees=45
  swing1Motion=Locked
  swing2LimitDegrees=45
  swing2Motion=Locked
  twistLimitDegrees=45
  twistMotion=Locked
 }
 {
  bone1Name=dyn_jacket_back_base_R
  bone2Name=pelvis
  swing1LimitDegrees=45
  swing1Motion=Locked
  swing2LimitDegrees=45
  swing2Motion=Locked
  twistLimitDegrees=45
  twistMotion=Locked
 }
 {bone1Name=dyn_jacket_back_02_R, bone2Name=dyn_jacket_back_base_R, swing1LimitDegrees=10, swing2LimitDegrees=10, twistLimitDegrees=0}
 {
  bone1Name=dyn_jacket_base_L
  bone2Name=pelvis
  swing1LimitDegrees=45
  swing1Motion=Locked
  swing2LimitDegrees=45
  swing2Motion=Locked
  twistLimitDegrees=45
  twistMotion=Locked
 }
 {bone1Name=dyn_jacket_01_L, bone2Name=dyn_jacket_base_L, swing1LimitDegrees=20, swing2LimitDegrees=5, twistLimitDegrees=0}
 {bone1Name=dyn_jacket_02_L, bone2Name=dyn_jacket_01_L, swing1LimitDegrees=20, swing2LimitDegrees=10, twistLimitDegrees=10}
 {
  bone1Name=dyn_jacket_base_R
  bone2Name=pelvis
  swing1LimitDegrees=45
  swing1Motion=Locked
  swing2LimitDegrees=45
  swing2Motion=Locked
  twistLimitDegrees=45
  twistMotion=Locked
 }
 {bone1Name=dyn_jacket_01_R, bone2Name=dyn_jacket_base_R, swing1LimitDegrees=10, swing2LimitDegrees=10, twistLimitDegrees=0}
 {bone1Name=dyn_jacket_02_R, bone2Name=dyn_jacket_01_R, swing1LimitDegrees=20, swing2LimitDegrees=20, twistLimitDegrees=5}
 {
  bone1Name=dyn_jacket_front_base_L
  bone2Name=pelvis
  swing1LimitDegrees=45
  swing1Motion=Locked
  swing2LimitDegrees=45
  swing2Motion=Locked
  twistLimitDegrees=45
  twistMotion=Locked
 }
 {bone1Name=dyn_jacket_front_01_L, bone2Name=dyn_jacket_front_base_L, swing1LimitDegrees=40, swing2LimitDegrees=10, twistLimitDegrees=5}
 {bone1Name=dyn_jacket_front_02_L, bone2Name=dyn_jacket_front_01_L, swing1LimitDegrees=40, swing2LimitDegrees=10, twistLimitDegrees=5}
 {
  bone1Name=dyn_jacket_front_base_R
  bone2Name=pelvis
  swing1LimitDegrees=45
  swing1Motion=Locked
  swing2LimitDegrees=45
  swing2Motion=Locked
  twistLimitDegrees=45
  twistMotion=Locked
 }
 {bone1Name=dyn_jacket_front_01_R, bone2Name=dyn_jacket_front_base_R, swing1LimitDegrees=40, swing2LimitDegrees=10, twistLimitDegrees=5}
 {bone1Name=dyn_jacket_front_02_R, bone2Name=dyn_jacket_front_01_R, swing1LimitDegrees=40, swing2LimitDegrees=10, twistLimitDegrees=5}
 {
  bone1Name=spine_01
  bone2Name=pelvis
  swing1LimitDegrees=45
  swing1Motion=Locked
  swing2LimitDegrees=45
  swing2Motion=Locked
  twistLimitDegrees=45
  twistMotion=Locked
 }
 {
  bone1Name=thigh_l
  bone2Name=pelvis
  swing1LimitDegrees=45
  swing1Motion=Locked
  swing2LimitDegrees=45
  swing2Motion=Locked
  twistLimitDegrees=45
  twistMotion=Locked
 }
 {
  bone1Name=thigh_r
  bone2Name=pelvis
  swing1LimitDegrees=45
  swing1Motion=Locked
  swing2LimitDegrees=45
  swing2Motion=Locked
  twistLimitDegrees=45
  twistMotion=Locked
 }
 {bone1Name=dyn_jacket_back_02_L, bone2Name=dyn_jacket_back_base_L, swing1LimitDegrees=20, swing2LimitDegrees=10, twistLimitDegrees=5}
 {
  bone1Name=dyn_jacket_back_base_C
  bone2Name=dyn_jacket_back_02_L
  swing1LimitDegrees=45
  swing1Motion=Locked
  swing2LimitDegrees=45
  swing2Motion=Locked
  twistLimitDegrees=45
  twistMotion=Locked
 }
]
