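# Physics asset for the Tie_02 mesh of the SewMesa medium male player body, for simulating the tie.
# Capsule bodies cover two spine bones and the tie bones: a root, one left segment and two chained right segments.
# Constraints lock the spine pair, tie root to upper spine and left segment to root.
# Only the right chain leaves its motions unlocked, with asymmetric swing limits and zero twist limit, so it can swing.
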PhysicsAsset /BRCosmetics/Characters/Player/Male/Medium/Bodies/M_MED_SewMesa/Meshes/M_MED_SewMesa_Tie_02_Physics_Asset
bodies {
 dyn_Tie_01 [{
   shapeType=Capsule
   length=4.5014
   radius=2.552
   shapeName=dyn_Tie_01_capsule
   center=<5.0058 -0.0575 0.6482>
   rotation=<rot 0 -91.5532 -84.6555>
  }]
 dyn_Tie_01_l [{shapeType=Capsule, length=12, radius=2, shapeName=dyn_Tie_01_l_capsule, center=<6.1665 -1.4851 0.5>, rotation=<rot 0 90 90>}]
 dyn_Tie_01_r [{shapeType=Capsule, length=6, radius=1, shapeName=dyn_Tie_01_r_capsule, center=<6 0.5 0.5>, rotation=<rot 0 90 90>}]
 dyn_Tie_02_r [{shapeType=Capsule, length=4, radius=1, shapeName=dyn_Tie_02_r_capsule, center=<6 0.5 0.5>, rotation=<rot 0 90 90>}]
 spine_01 [{
   shapeType=Capsule
   length=14.1286
   radius=11.9903
   shapeName=spine_01_capsule
   center=<9.9607 -0.0014 -0.2579>
   rotation=<rot 0 -90.0084 90.976>
  }]
 spine_04 [{
   shapeType=Capsule
   length=10.8277
   radius=12.9799
   shapeName=spine_04_capsule
   center=<6.5277 -1.7092 0.3077>
   rotation=<rot -0.602 -99.9899 87.3222>
  }]
}
constraints [
 {
  bone1Name=spine_04
  bone2Name=spine_01
  swing1LimitDegrees=45
  swing1Motion=Locked
  swing2LimitDegrees=45
  swing2Motion=Locked
  twistLimitDegrees=45
  twistMotion=Locked
 }
 {
  bone1Name=dyn_Tie_01_l
  bone2Name=dyn_Tie_01
  swing1LimitDegrees=40
  swing1Motion=Locked
  swing2LimitDegrees=40
  swing2Motion=Locked
  twistLimitDegrees=0
  twistMotion=Locked
 }
 {bone1Name=dyn_Tie_01_r, bone2Name=dyn_Tie_01, swing1LimitDegrees=10, swing2LimitDegrees=50, twistLimitDegrees=0}
 {bone1Name=dyn_Tie_02_r, bone2Name=dyn_Tie_01_r, swing1LimitDegrees=20, swing2LimitDegrees=50, twistLimitDegrees=0}
 {
  bone1Name=dyn_Tie_01
  bone2Name=spine_04
  swing1LimitDegrees=45
  swing1Motion=Locked
  swing2LimitDegrees=45
  swing2Motion=Locked
  twistLimitDegrees=45
  twistMotion=Locked
 }
]
