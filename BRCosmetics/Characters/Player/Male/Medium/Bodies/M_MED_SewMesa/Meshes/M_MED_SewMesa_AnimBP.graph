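# Post-process-style cosmetic Animation Blueprint for the M_MED_SewMesa body, adding secondary motion to its coat and ties.
# Copies the parent mesh pose through a Control Rig, weighted jacket constraints, three RigidBody simulations, then parachute and crouch bone offsets.
# EventGraph derives coat gravity through a deprecated helper, plus melee, thigh-bend and movement values the AnimGraph never visibly reads.
# The skydiving jacket offsets are disconnected from the output pose.
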AnimBlueprint /BRCosmetics/Characters/Player/Male/Medium/Bodies/M_MED_SewMesa/Meshes/M_MED_SewMesa_AnimBP
target_skeleton=&/M_MED_SewMesa_Skeleton
graphs [
 AnimGraph {nodes [
   n1 = Get bIsSkyDiving()
   n2 = "Transform (Modify) Bone - Bone: dyn_jacket_front_base_l"!(ComponentPose, bAlphaBoolEnabled:bool=@n1, Translation="0,0,0", Rotation="0,0,-5", Scale="1,1,1") -> (Pose) {
    alphaInputType=Bool
    rotationMode=BMM_Additive
    translationMode=BMM_Additive
    alphaBoolBlend {blendOption=Linear}
    boneToModify {boneName=dyn_jacket_front_base_l}
    rotation {roll=-5}
   }
   "Transform (Modify) Bone - Bone: dyn_jacket_front_base_r"!(ComponentPose=@n2, bAlphaBoolEnabled:bool=@n1, Translation="0,0,0", Rotation="0,0,-5", Scale="1,1,1") -> (Pose) {
    alphaInputType=Bool
    rotationMode=BMM_Additive
    translationMode=BMM_Additive
    alphaBoolBlend {blendOption=Linear}
    boneToModify {boneName=dyn_jacket_front_base_r}
    rotation {roll=-5}
   }
   n3 = Control Rig(Source=Copy Pose From Mesh() {bUseAttachedParent=true}) {
    alphaBoolBlend {blendOption=Linear}
    controlRigAssetReference {blueprintRigClass=&/M_MED_SewMesa_CtrlRig.M_MED_SewMesa_CtrlRig_C}
   }
   n4 = "Constraint - dyn_jacket_front_01_l (PPPPP)"!(
    ComponentPose=Local To Component(LocalPose=@n3)
    Alpha=1
    ConstraintWeights_0=0.292
    ConstraintWeights_1:float=0.2777
    ConstraintWeights_2:float=0.1984
    ConstraintWeights_3:float=0.1273
    ConstraintWeights_4:float=0.1047
   ) -> (Pose) {
    alphaBoolBlend {blendOption=Linear}
    boneToModify {boneName=dyn_jacket_front_01_l}
    constraintSetup [
     {targetBone {boneName=spine_01}}
     {targetBone {boneName=dyn_jacket_front_01_l}}
     {targetBone {boneName=pelvis}}
     {targetBone {boneName=spine_02}}
     {targetBone {boneName=spine_03}}
    ]
    constraintWeights [1,1,1,1,1]
   }
   n5 = "Constraint - dyn_jacket_front_01_r (PPPPP)"!(
    ComponentPose=@n4
    Alpha=1
    ConstraintWeights_0=0.285
    ConstraintWeights_1:float=0.2745
    ConstraintWeights_2:float=0.2071
    ConstraintWeights_3:float=0.1328
    ConstraintWeights_4:float=0.1006
   ) -> (Pose) {
    alphaBoolBlend {blendOption=Linear}
    boneToModify {boneName=dyn_jacket_front_01_r}
    constraintSetup [
     {targetBone {boneName=spine_01}}
     {targetBone {boneName=dyn_jacket_front_01_r}}
     {targetBone {boneName=pelvis}}
     {targetBone {boneName=spine_02}}
     {targetBone {boneName=spine_03}}
    ]
    constraintWeights [1,1,1,1,1]
   }
   n6 = RigidBody!(ComponentPose=@n5, Alpha, OverrideWorldGravity:Vector=Get Gravity Override(), ExternalForce) -> (Pose) {
    bOverrideWorldGravity=true
    overridePhysicsAsset=&/M_MED_SewMesa_Coat_Physics_Asset
    simulationSpace=BaseBoneSpace
    alphaBoolBlend {blendOption=Linear}
    baseBoneRef {boneName=pelvis}
    overrideWorldGravity {z=-980}
    simSpaceSettings {maxAngularAcceleration=400, maxAngularVelocity=10, maxLinearAcceleration=2000, maxLinearVelocity=600, velocityScaleZ=0.8, worldAlpha=0.6}
   }
   n7 = RigidBody(ComponentPose=@n6, OverrideWorldGravity="0,0,-980") {
    bOverrideWorldGravity=true
    overridePhysicsAsset=&/M_MED_SewMesa_Tie_Physics_Asset
    simulationSpace=BaseBoneSpace
    alphaBoolBlend {blendOption=Linear}
    baseBoneRef {boneName=spine_04}
    overrideWorldGravity {z=-700}
    simSpaceSettings {maxAngularAcceleration=300, maxAngularVelocity=60, maxLinearAcceleration=6000, maxLinearVelocity=1200, worldAlpha=0.8}
   }
   n8 = RigidBody(ComponentPose=@n7, OverrideWorldGravity="0,0,-980") {
    bOverrideWorldGravity=true
    overridePhysicsAsset=&/M_MED_SewMesa_Tie_02_Physics_Asset
    simulationSpace=BaseBoneSpace
    alphaBoolBlend {blendOption=Linear}
    baseBoneRef {boneName=spine_04}
    overrideWorldGravity {z=-700}
    simSpaceSettings {maxAngularAcceleration=300, maxAngularVelocity=60, maxLinearAcceleration=6000, maxLinearVelocity=1200, worldAlpha=0.8}
   }
   n9 = Get bIsParachuteOpened()
   n10 = "Transform (Modify) Bone - Bone: dyn_jacket_front_base_l"(ComponentPose=@n8, bAlphaBoolEnabled=@n9, Translation="0,0,0", Rotation="0,0,-10", Scale="1,1,1") {
    alphaInputType=Bool
    rotationMode=BMM_Additive
    translationMode=BMM_Additive
    alphaBoolBlend {blendOption=Linear}
    boneToModify {boneName=dyn_jacket_front_base_l}
    rotation {roll=-10}
   }
   n11 = "Transform (Modify) Bone - Bone: dyn_jacket_front_base_r"(ComponentPose=@n10, bAlphaBoolEnabled=@n9, Translation="0,0,0", Rotation="0,0,-5", Scale="1,1,1") {
    alphaInputType=Bool
    rotationMode=BMM_Additive
    translationMode=BMM_Additive
    alphaBoolBlend {blendOption=Linear}
    boneToModify {boneName=dyn_jacket_front_base_r}
    rotation {roll=-5}
   }
   n12 = "Transform (Modify) Bone - Bone: dyn_Pouch_01_l"!(ComponentPose=@n11, bAlphaBoolEnabled:bool=Get bIsCrouching(), Translation="0,1.2,-1.2", Rotation="0,0,0", Scale="1,1,1") -> (Pose) {
    alphaInputType=Bool
    rotationMode=BMM_Additive
    translationMode=BMM_Additive
    alphaBoolBlend {blendOption=Linear}
    boneToModify {boneName=dyn_Pouch_01_l}
    translation {y=1.2, z=-1.2}
   }
   Output Pose(Result=Component To Local(ComponentPose=@n12)) {layerGroup=None}
  ]}
 EventGraph {nodes [
   Event BlueprintUpdateAnimation()
   > n1 = Sequence!(execute) -> (then_0, then_1, then_2:Exec, then_3:Exec, then_4:Exec)
   > n2 = Branch(Condition=Get bIsPlayingMeleeAnim())
   > Set pickaxeSwing(pickaxeSwing:double=1) -> (Output_Get:double)
   Set pickaxeSwing(execute=@n2.else, pickaxeSwing=0)
   Is Valid(exec=@n1.then_3, InputObject=Get Skel Mesh())
   > n3 = Gravity Override State Function DEPRECATED(CharacterPartAnimInstance=Self-Reference(), GameTimeInSeconds=GetGameTimeInSeconds(), GravityOverrideParams=Get Gravity Override Params())
   > Set Gravity Override(Gravity Override=@n3)
   n4 = SelectVector(A="0.1,0.35,0.35", B="1,1,0.6", bPickA=OR Boolean(A=Get bIsSkyDiving(), B=Get bIsWaterJump() -> (bIsWaterJump:bool)))
   > Set Component Linear Acc Scale(Component Linear Acc Scale:Vector=@n4) -> (Output_Get:Vector)
   > Set Component Linear Vel Scale(Component Linear Vel Scale:Vector=@n4) -> (Output_Get:Vector)
   n5 = Cast To PlayerPawn_Athena(execute=@n1.then_4, Object=TryGetPawnOwner())
   > Set head_mesh(head_mesh:Skeletal Mesh Component Object Reference=GetSkeletalMeshForPartType(self=@n5, PartType=Head)) -> (Output_Get:Skeletal Mesh Component Object Reference)
   n6 = GetOwningActor()
   n7 = GetVelocity!(self:Actor Object Reference=@n6) -> (ReturnValue:Vector)
   n8 = GetActorForwardVector!(self:Actor Object Reference=@n6) -> (ReturnValue:Vector)
   Set ForwardMovementAlpha(
    execute=@n1.then_2
    ForwardMovementAlpha:double=MapRangeClamped(Value=Dot Product!(A:Vector=@n7, B:Vector=@n8) -> (ReturnValue:double), InRangeA=0, InRangeB=400)
   ) -> (Output_Get:double)
   > Set FloatConstraint_alpha(FloatConstraint_alpha:double=SelectFloat(A=0.4, B=0, bPickA=Get bIsSkydiveFloating())) -> (Output_Get:double)
   n9 = Get Skel Mesh()
   Set LeftLegForwardBendCompensationAlpha(
    execute=@n1.then_1
    LeftLegForwardBendCompensationAlpha:double=GetAlphaFromReferenceTransformDelta(SkeletalMeshComponent=@n9, BoneName=thigh_l, BaseName=pelvis, InputRange=(X=-17,Y=-103), TwistAxis="1,0,0")
   ) -> (Output_Get:double)
   > Set RightLegForwardBendCompensationAlpha(RightLegForwardBendCompensationAlpha:double=GetAlphaFromReferenceTransformDelta(SkeletalMeshComponent=@n9, BoneName=thigh_r, BaseName=pelvis, InputRange=(X=-17,Y=-103), TwistAxis="1,0,0")) -> (Output_Get:double)
   > Set RThighFrontUpAlpha(RThighFrontUpAlpha:double=float + float(
     A=GetAlphaFromReferenceTransformDelta(SkeletalMeshComponent=@n9, BoneName=thigh_r, BaseName=pelvis, InputRange=(X=0,Y=-50), TwistAxis="1,0,0")
     B=GetAlphaFromReferenceTransformDelta(SkeletalMeshComponent=@n9, BoneName=thigh_r, BaseName=pelvis, InputRange=(X=-70,Y=-110), OutputRange=(X=0,Y=-1), TwistAxis="1,0,0")
    )) -> (Output_Get:double)
   > Set LThighFrontUpAlpha(LThighFrontUpAlpha:double=float + float(
     A=GetAlphaFromReferenceTransformDelta(SkeletalMeshComponent=@n9, BoneName=thigh_l, BaseName=pelvis, InputRange=(X=0,Y=-60), TwistAxis="1,0,0")
     B=GetAlphaFromReferenceTransformDelta(SkeletalMeshComponent=@n9, BoneName=thigh_l, BaseName=pelvis, InputRange=(X=-70,Y=-110), OutputRange=(X=0,Y=-1), TwistAxis="1,0,0")
    )) -> (Output_Get:double)
  ]}
]
variables [
 Component Linear Acc Scale {type=Vector}
 Component Linear Vel Scale {type=Vector}
 FloatConstraint_alpha {type=number}
 ForwardMovementAlpha {type=number}
 Gravity Override {type=Vector}
 Gravity Override Params {type=GravityOverrideParamsStruct}
 head_mesh {type=SkeletalMeshComponent}
 LeftLegForwardBendCompensationAlpha {type=number}
 LThighFrontUpAlpha {type=number}
 Mat Instance Dynamic {type=MaterialInstanceDynamic}
 Overlay Mat Instance Dynamic {type=MaterialInstanceDynamic}
 pickaxeSwing {type=number}
 RightLegForwardBendCompensationAlpha {type=number}
 RThighFrontUpAlpha {type=number}
 Skel Mesh {type=SkeletalMeshComponent}
]
