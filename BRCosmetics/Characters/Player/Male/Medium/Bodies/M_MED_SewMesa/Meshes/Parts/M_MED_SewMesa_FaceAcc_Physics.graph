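# Physics asset for the face accessory mesh of the male medium SewMesa player cosmetic body, setting up six dynamic hair pieces on the head.
# A large head capsule anchors six small base-bone capsules, each constrained to it with locked swing and twist.
# Each base bone links to its own hair bone capsule through an unlocked joint with small swing limit values, forming six independent two-bone chains.
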PhysicsAsset /BRCosmetics/Characters/Player/Male/Medium/Bodies/M_MED_SewMesa/Meshes/Parts/M_MED_SewMesa_FaceAcc_Physics
bodies {
 dyn_Base_Hair_02 [{shapeType=Capsule, length=0.505, radius=0.505, shapeName=dyn_Base_Hair_02_capsule, rotation=<rot 0 -90 0>}]
 dyn_Base_Hair_03 [{shapeType=Capsule, length=0.505, radius=0.505, shapeName=dyn_Base_Hair_03_capsule, rotation=<rot 0 -90 0>}]
 dyn_Base_Hair_04 [{shapeType=Capsule, length=0.505, radius=0.505, shapeName=dyn_Base_Hair_04_capsule, rotation=<rot 0 -90 0>}]
 dyn_Base_Hair_05 [{shapeType=Capsule, length=0.505, radius=0.505, shapeName=dyn_Base_Hair_05_capsule, rotation=<rot 0 -90 0>}]
 dyn_Base_Hair_06 [{shapeType=Capsule, length=0.505, radius=0.505, shapeName=dyn_Base_Hair_06_capsule, rotation=<rot 0 -90 0>}]
 dyn_Hair_01 [{shapeType=Capsule, length=0.4964, radius=0.4734, shapeName=dyn_Hair_01_capsule, center=<2.0112 -0.1227 0.0671>, rotation=<rot 0 90 -90>}]
 dyn_Hair_02 [{
   shapeType=Capsule
   length=2
   radius=1
   shapeName=dyn_Hair_02_capsule
   center=<0.0911 0.1882 0.2511>
   rotation=<rot -67.6067 -127.0282 121.5701>
  }]
 dyn_Hair_03 [{
   shapeType=Capsule
   length=1.4285
   radius=0.8764
   shapeName=dyn_Hair_03_capsule
   center=<1.7479 0.0594 0.0168>
   rotation=<rot 0.0005 -91.0234 100.6904>
  }]
 dyn_Hair_04 [{shapeType=Capsule, length=1, radius=1, shapeName=dyn_Hair_04_capsule, center=<2.7214 0.0395 0.0319>, rotation=<rot 0 80.1663 -62.4714>}]
 dyn_Hair_05 [{shapeType=Capsule, length=1, radius=1, shapeName=dyn_Hair_05_capsule, center=<0.8415 0.153 0.0688>, rotation=<rot 0 36.0393 -72.5923>}]
 dyn_Hair_06 [{shapeType=Capsule, length=1, radius=0.8, shapeName=dyn_Hair_06_capsule, center=<2.4049 -0.0314 0.0838>, rotation=<rot 0 90 90>}]
 dyn_base_Hair_01 [{shapeType=Capsule, length=0.505, radius=0.505, shapeName=dyn_base_Hair_01_capsule, rotation=<rot 0 -90 0>}]
 head [{shapeType=Capsule, length=4.6339, radius=7.8451, center=<5.7724 4.6809 -0.0112>, rotation=<rot -0.0933 93.4975 -89.9074>}]
}
constraints [
 {
  bone1Name=dyn_Base_Hair_02
  bone2Name=head
  swing1LimitDegrees=20
  swing1Motion=Locked
  swing2LimitDegrees=20
  swing2Motion=Locked
  twistLimitDegrees=0
  twistMotion=Locked
 }
 {bone1Name=dyn_Hair_02, bone2Name=dyn_Base_Hair_02, swing1LimitDegrees=10, swing2LimitDegrees=10, twistLimitDegrees=0}
 {
  bone1Name=dyn_Base_Hair_03
  bone2Name=head
  swing1LimitDegrees=45
  swing1Motion=Locked
  swing2LimitDegrees=45
  swing2Motion=Locked
  twistLimitDegrees=45
  twistMotion=Locked
 }
 {bone1Name=dyn_Hair_03, bone2Name=dyn_Base_Hair_03, swing1LimitDegrees=10, swing2LimitDegrees=10, twistLimitDegrees=0}
 {
  bone1Name=dyn_Base_Hair_04
  bone2Name=head
  swing1LimitDegrees=50
  swing1Motion=Locked
  swing2LimitDegrees=50
  swing2Motion=Locked
  twistLimitDegrees=10
  twistMotion=Locked
 }
 {bone1Name=dyn_Hair_04, bone2Name=dyn_Base_Hair_04, swing1LimitDegrees=10, swing2LimitDegrees=10, twistLimitDegrees=0}
 {
  bone1Name=dyn_Base_Hair_06
  bone2Name=head
  swing1LimitDegrees=45
  swing1Motion=Locked
  swing2LimitDegrees=45
  swing2Motion=Locked
  twistLimitDegrees=45
  twistMotion=Locked
 }
 {bone1Name=dyn_Hair_06, bone2Name=dyn_Base_Hair_06, swing1LimitDegrees=5, swing2LimitDegrees=8, twistLimitDegrees=0}
 {
  bone1Name=dyn_base_Hair_01
  bone2Name=head
  swing1LimitDegrees=45
  swing1Motion=Locked
  swing2LimitDegrees=45
  swing2Motion=Locked
  twistLimitDegrees=45
  twistMotion=Locked
 }
 {bone1Name=dyn_Hair_01, bone2Name=dyn_base_Hair_01, swing1LimitDegrees=10, swing2LimitDegrees=10, twistLimitDegrees=0}
 {
  bone1Name=dyn_Base_Hair_05
  bone2Name=head
  swing1LimitDegrees=45
  swing1Motion=Locked
  swing2LimitDegrees=45
  swing2Motion=Locked
  twistLimitDegrees=45
  twistMotion=Locked
 }
 {bone1Name=dyn_Hair_05, bone2Name=dyn_Base_Hair_05, swing1LimitDegrees=10, swing2LimitDegrees=10, twistLimitDegrees=0}
]
